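# Translucent UI material that procedurally draws a stack indicator for the Hawaii item: a filled circle with inner glow, pulsing glow and outer ring.
# MF_UI_Circle SDFs shape the fill, glow and ring, while MF_UI_RadialGradient masks the inner glow.
# A time-driven sine modulates the glow falloff exponent.
# Layers blend by mask, with Stack Opacity scaling circle and glow alpha before the separately weighted ring is added.
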Material /StaminaGameplay/Gameplay/Items/Hawaii/UI/M_UI_HawaiiStack
graph {
 properties {blendMode=BLEND_Translucent, materialDomain=MD_UI, pixelDepthOffsetMode=PDOM_Legacy, usedShadingModels=Default Lit}
 nodes [
  n1 = VectorParameter() {group=1 - Circle, parameterName=Fill Color, sortPriority=0, defaultValue=<rgba 0.8185 0.9823 0.9798 1>}
  n2 = VectorParameter() {group=1 - Circle, parameterName=Inner Glow Color, sortPriority=1, defaultValue=<rgba 0 0.5138 0.6302 1>}
  n3 = NamedRerouteUsage()
  n4 = ScalarParameter() {defaultValue=1, group=1 - Circle, parameterName=Inner Glow Opacity, sliderMax=1, sortPriority=2}
  n5 = LinearInterpolate(
   A=VectorParameter() {group=2 - Glow, parameterName=Glow Color, sortPriority=0, defaultValue=<rgba 0 0.2418 0.6302 1>}.RGB
   B=LinearInterpolate(A=@n1.RGB, B=@n2.RGB, Alpha=Multiply(A=@n3, B=@n4))
   Alpha=NamedRerouteUsage()
  )
  n6 = VectorParameter() {group=3 - Outer Ring, parameterName=Outer Ring Color, sortPriority=0, defaultValue=<rgba 0.9413 0.9823 0.9816 1>}
  LinearInterpolate(
   A=LinearInterpolate(A=@n5, B=@n6.RGB, Alpha=NamedRerouteUsage())
   B=LinearInterpolate(A=@n5, B=@n6.RGB, Alpha=NamedRerouteUsage())
   Alpha=ScalarParameter() {defaultValue=1, group=0 - Global, parameterName=Stack Opacity, sliderMax=1, sortPriority=0}
  )
  n7 = OneMinus(None=MaterialFunctionCall(Size=ScalarParameter() {defaultValue=0.6, group=1 - Circle, parameterName=Circle Size, sliderMax=1, sortPriority=3}) {materialFunction=&MF_UI_Circle}.SDF)
  n8 = Multiply(A=Time(), B=ScalarParameter() {defaultValue=0.4, desc=Pulsing Speed, group=0 - Global, parameterName=Pulsing Speed}) {constB=0.4}
  n9 = Sine(None=ConstantBiasScale(None=Frac(None=@n8)) {bias=2})
  n10 = Power(
   Base=Subtract(A=@n7, B=ScalarParameter() {defaultValue=0.25, group=2 - Glow, parameterName=Glow Size, sortPriority=2}) {constB=0.25}
   Exp=LinearInterpolate(Alpha=NamedRerouteDeclaration(None=@n9) {name=Pulsing, nodeColor=<rgba 0 0 0 1>}) {constA=1, constB=1.5}
  ) {constExponent=1}
  n11 = OneMinus(None=NamedRerouteUsage())
  NamedRerouteDeclaration(None=Saturate(None=Multiply(A=@n10, B=@n11))) {name=Glow, nodeColor=<rgba 0 0 0 1>}
  n12 = ScalarParameter() {defaultValue=1, group=3 - Outer Ring, parameterName=Outer Circle Feather Start, sortPriority=4}
  n13 = ScalarParameter() {defaultValue=0.02, group=3 - Outer Ring, parameterName=Outer Circle Feather End, sortPriority=5}
  n14 = OneMinus(None=MaterialFunctionCall(
    Size=ScalarParameter() {defaultValue=0.9, group=3 - Outer Ring, parameterName=Outer Circle Size, sliderMax=1, sortPriority=2}
    OutlineThickness=ScalarParameter() {defaultValue=0.03, group=3 - Outer Ring, parameterName=Outer Circle Thickness, sliderMax=1, sortPriority=3}
   ) {materialFunction=&MF_UI_Circle}.SDFOutline)
  NamedRerouteDeclaration(None=SmoothStep(Min=Subtract(A=@n12, B=@n13) {constB=0.05}, Max=@n12, Value=Saturate(None=@n14))) {name=Outer Ring, nodeColor=<rgba 0 0 0 1>}
  n15 = ScalarParameter() {defaultValue=0.5, group=1 - Circle, parameterName=Inner Glow Start}
  n16 = LinearInterpolate(Alpha=NamedRerouteUsage()) {constA=0.7, constB=0.8}
  n17 = NamedRerouteDeclaration(None=OneMinus(None=MaterialFunctionCall(GradientStart=@n15, GradientEnd=@n16) {materialFunction=&MF_UI_RadialGradient}.Smooth)) {name=Inner Glow Mask, nodeColor=<rgba 0 0 0 1>}
  n18 = ScalarParameter() {defaultValue=1, group=1 - Circle, parameterName=Circle Feather Start, sortPriority=4}
  n19 = ScalarParameter() {defaultValue=0.02, group=1 - Circle, parameterName=Circle Feather End, sortPriority=5}
  n20 = NamedRerouteDeclaration(None=SmoothStep(Min=Subtract(A=@n18, B=@n19) {constB=0.05}, Max=@n18, Value=Saturate(None=@n7))) {name=Fill, nodeColor=<rgba 0 0 0 1>}
  NamedRerouteDeclaration(None=Saturate(None=Multiply(A=@n17, B=@n20))) {name=Inner Glow, nodeColor=<rgba 0 0 0 1>}
  n21 = NamedRerouteUsage()
  n22 = NamedRerouteDeclaration(None=Multiply(A=NamedRerouteUsage(), B=ScalarParameter() {defaultValue=1, group=2 - Glow, parameterName=Glow Opacity, sliderMax=1, sortPriority=1})) {name=Glow Opacity, nodeColor=<rgba 0 0 0 1>}
  n23 = Saturate(None=Saturate(None=Add(A=@n21, B=@n22)))
  n24 = ScalarParameter() {defaultValue=1, group=0 - Global, parameterName=Stack Opacity, sliderMax=1, sortPriority=0}
  n25 = Multiply(
   A=ScalarParameter() {defaultValue=0.25, group=3 - Outer Ring, parameterName=Outer Ring Opacity, sliderMax=1, sortPriority=1}
   B=NamedRerouteUsage()
  )
  Saturate(None=Add(A=Multiply(A=@n23, B=@n24), B=NamedRerouteDeclaration(None=@n25) {name=Outer Ring Opacity, nodeColor=<rgba 0 0 0 1>}))
  ScalarParameter() {defaultValue=0.825, parameterName=Inner Glow End}
 ]
}
