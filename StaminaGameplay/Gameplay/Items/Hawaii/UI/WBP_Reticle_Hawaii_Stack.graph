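# Reticle widget for the Hawaii item that shows a single stack indicator as layered StackVisual and Rays material images.
# OnInit sets the stack and outer ring opacity parameters to visible values, while OnRecharge sets both to zero.
# OnUse plays Anim_Use, a short burst that animates material parameters on both images, scales and rotates the rays, and fades the stack out.
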WidgetBlueprint /StaminaGameplay/Gameplay/Items/Hawaii/UI/WBP_Reticle_Hawaii_Stack
animations [Anim_Use {
  binding_count=2
  length_seconds=0.6
  object=&self:Anim_Use
  sequence {bindings [
    Rays {
     id=#1
     possessed_class=Image
     tracks [
      {
       type=MovieSceneWidgetMaterialTrack
       sections [{
         type=MovieSceneParameterSection
         channels [
          Global Opacity {type=MovieSceneScriptingFloatChannel, keys [[0,0],[2,1],[5,1],[8,0]]}
          Primary Rotation {type=MovieSceneScriptingFloatChannel, keys [[0,0.032],[12,0.062]]}
          Secondary Rotation {type=MovieSceneScriptingFloatChannel, keys [[0,0.25],[12,0.19]]}
          Primary Opacity {type=MovieSceneScriptingFloatChannel, keys [[0,0.5],[2,0.5],[12,0.75]]}
          Secondary Opacity {type=MovieSceneScriptingFloatChannel, keys [[0,0.15],[2,0.15],[12,0.25]]}
          ColorR {type=MovieSceneScriptingFloatChannel, keys [[0,0],[2,0],[5,0.5333]]}
          ColorG {type=MovieSceneScriptingFloatChannel, keys [[0,0.5149],[2,0.5149],[5,0.8228]]}
          ColorB {type=MovieSceneScriptingFloatChannel, keys [[0,0.6308],[2,0.6308],[5,0.855]]}
          ColorA {type=MovieSceneScriptingFloatChannel, const [1,3]}
          Edge ColorR {type=MovieSceneScriptingFloatChannel, keys [[0,0],[2,1],[5,1]]}
          Edge ColorG {type=MovieSceneScriptingFloatChannel, keys [[0,0.5137],[2,1],[5,1]]}
          Edge ColorB {type=MovieSceneScriptingFloatChannel, keys [[0,0.6314],[2,1],[5,1]]}
          Edge ColorA {type=MovieSceneScriptingFloatChannel, const [1,3]}
         ]
        }]
      }
      {
       type=MovieScene2DTransformTrack
       sections [{
         type=MovieScene2DTransformSection
         channels [
          Scale.X {type=MovieSceneScriptingFloatChannel, keys [[0,2],[5,4],[12,2]]}
          Scale.Y {type=MovieSceneScriptingFloatChannel, keys [[0,2],[5,4],[12,2]]}
         ]
        }]
      }
     ]
    }
    StackVisual {
     id=#2
     possessed_class=Image
     tracks [{
       type=MovieSceneWidgetMaterialTrack
       sections [{
         type=MovieSceneParameterSection
         channels [
          Stack Opacity {type=MovieSceneScriptingFloatChannel, thinned_from=4, keys [[0,1],[6,1],[11,0]]}
          Outer Ring Opacity {type=MovieSceneScriptingFloatChannel, keys [[0,0.25],[1,0.243],[5,0]]}
          Circle Size {type=MovieSceneScriptingFloatChannel, thinned_from=6, keys [[0,0.5],[5,0.85],[8,0.85],[11,0],[12,0.5]]}
          Inner Glow Start {type=MovieSceneScriptingFloatChannel, keys [[0,0.5],[5,0],[6,-1.25],[11,-1.25],[12,0.5]]}
          Glow Size {type=MovieSceneScriptingFloatChannel, keys [[0,0.5],[2,0.8],[6,0.8],[12,0.5]]}
         ]
        }]
      }]
    }
   ]}
 }]
graphs [EventGraph {nodes [
   OnInit() {is=title K2Node_CustomEvent}
   > n1 = GetDynamicMaterial(self=Get StackVisual() -> (StackVisual:Image Object Reference))
   > SetScalarParameterValue(self=@n1, ParameterName=Stack Opacity, Value=1)
   > SetScalarParameterValue(self=@n1, ParameterName=Outer Ring Opacity, Value=0.25)
   OnUse() {is=title K2Node_CustomEvent}
   > PlayAnimation(InAnimation=Get Anim_Use() -> (Anim_Use:Widget Animation Object Reference))
   OnRecharge() {is=title K2Node_CustomEvent}
   > n2 = GetDynamicMaterial(self=Get StackVisual())
   > SetScalarParameterValue(self=@n2, ParameterName=Stack Opacity, Value=0)
   > SetScalarParameterValue(self=@n2, ParameterName=Outer Ring Opacity, Value=0)
  ]}]
variables []
widgets {
 info {inheritedWidgetCount=0, parentClass=&UserWidget, rootWidgetClass=&Overlay, widgetCount=4}
 widgets [
  Overlay_0 {widgetClassPath=&Overlay, widget=Overlay_0}
  Rays {
   widgetClassPath=&Image
   bIsVariable=true
   parent=Overlay_0
   widget=Rays
   properties {
    brush {
     bIsDynamicallyLoaded=false
     drawAs=Image
     imageType=NoImage
     mirroring=NoMirror
     resourceName=None
     resourceObject=&/MI_UI_HawaiiRays
     tiling=NoTile
     imageSize=<24 24>
     margin=<ltrb 0 0 0 0>
     outlineSettings {
      bUseBrushTransparency=false
      roundingType=HalfHeightRadius
      width=0
      color {colorUseRule=UseColor_Specified, specifiedColor=<rgba 0 0 0 0>}
      cornerRadii=<0 0 0 0>
     }
     tintColor {colorUseRule=UseColor_Specified, specifiedColor=<rgba 1 1 1 1>}
     uVRegion {bIsValid=false, max=<0 0>, min=<0 0>}
    }
    renderTransform {angle=0, scale=<2 2>, shear=<0 0>, translation=<0 0>}
   }
   slot {class=OverlaySlot, properties {horizontalAlignment=HAlign_Fill, verticalAlignment=VAlign_Fill}}
  }
  StackVisual {
   widgetClassPath=&Image
   bIsVariable=true
   parent=Overlay_0
   widget=StackVisual
   properties {brush {
     bIsDynamicallyLoaded=false
     drawAs=Image
     imageType=NoImage
     mirroring=NoMirror
     resourceName=None
     resourceObject=&/MI_UI_HawaiiStack
     tiling=NoTile
     imageSize=<24 24>
     margin=<ltrb 0 0 0 0>
     outlineSettings {
      bUseBrushTransparency=false
      roundingType=HalfHeightRadius
      width=0
      color {colorUseRule=UseColor_Specified, specifiedColor=<rgba 0 0 0 0>}
      cornerRadii=<0 0 0 0>
     }
     tintColor {colorUseRule=UseColor_Specified, specifiedColor=<rgba 1 1 1 1>}
     uVRegion {bIsValid=false, max=<0 0>, min=<0 0>}
    }}
   slot {class=OverlaySlot, properties {horizontalAlignment=HAlign_Fill, verticalAlignment=VAlign_Fill}}
  }
  Spacer_54 {
   widgetClassPath=&Spacer
   parent=Overlay_0
   widget=Spacer_54
   properties {size=<32 32>}
   slot {class=OverlaySlot, properties {horizontalAlignment=HAlign_Fill, verticalAlignment=VAlign_Fill}}
  }
 ]
}
