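# Reticle extension widget for the Stamina "Hawaii" beam weapon, showing charge progress, target-in-range state and three ammo stacks.
# On Construct it binds to weapon equip, charge, fire and finish delegates, plus the instigator's Hawaii component target and cooldown events.
# Animations drive the reticle's material parameters, with charge playback speed derived from a curve-table duration plus an offset.
# Ammo count sets each stack's used or recharged state and rotates the stack overlay.
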WidgetBlueprint /StaminaGameplay/Gameplay/Items/Hawaii/UI/WBP_Reticle_Hawaii
animations [
 Anim_Intro {
  binding_count=1
  length_seconds=0.3
  object=&self:Anim_Intro
  sequence {bindings [Reticle {
     id=#1
     possessed_class=Overlay
     tracks [
      {
       type=MovieSceneFloatTrack
       display_name=Render Opacity
       sections [{type=MovieSceneFloatSection, channels [None {type=MovieSceneScriptingFloatChannel, keys [[0,0],[6,1]]}]}]
      }
      {
       type=MovieScene2DTransformTrack
       sections [{
         type=MovieScene2DTransformSection
         channels [
          Translation.X {type=MovieSceneScriptingFloatChannel}
          Translation.Y {type=MovieSceneScriptingFloatChannel}
          Angle {type=MovieSceneScriptingFloatChannel}
          Scale.X {type=MovieSceneScriptingFloatChannel, keys [[0,1.25],[6,1]]}
          Scale.Y {type=MovieSceneScriptingFloatChannel, keys [[0,1.25],[6,1]]}
          Shear.X {type=MovieSceneScriptingFloatChannel}
          Shear.Y {type=MovieSceneScriptingFloatChannel}
         ]
        }]
      }
     ]
    }]}
 }
 Anim_Charge {
  binding_count=1
  length_seconds=2.75
  object=&self:Anim_Charge
  sequence {
   bindings [ReticleVisual {
     id=#2
     possessed_class=Image
     tracks [{
       type=MovieSceneWidgetMaterialTrack
       sections [{
         type=MovieSceneParameterSection
         channels [
          Triangles Offset {type=MovieSceneScriptingFloatChannel, thinned_from=5, keys [[0,0.275],[117,0.1],[165,0.1],[168,0.275]]}
          Triangle Opacity {type=MovieSceneScriptingFloatChannel, thinned_from=7, keys [[0,0.5],[117,1],[129,0.25],[141,1],[153,0.25],[165.001,1]]}
          Global Rotation {type=MovieSceneScriptingFloatChannel, thinned_from=4, keys [[0,0],[117,0],[165,-0.25]]}
          Circle Thickness {type=MovieSceneScriptingFloatChannel, keys [[0,0.015],[60,0.015],[117,0.035],[165,0.035],[168,0.015]]}
          Circle Size {type=MovieSceneScriptingFloatChannel, keys [[0,0.825],[60,0.825],[117,0.8],[165,0.8],[168,0.825]]}
         ]
        }]
      }]
    }]
   display_rate {numerator=60}
  }
 }
 Anim_Hide {
  binding_count=1
  length_seconds=0.4
  object=&self:Anim_Hide
  sequence {
   bindings [ReticleVisual {
     id=#3
     possessed_class=Image
     tracks [{
       type=MovieSceneWidgetMaterialTrack
       sections [{
         type=MovieSceneParameterSection
         end_frame=24
         start_frame=0
         channels [Global Opacity {type=MovieSceneScriptingFloatChannel, keys [[0,1],[24,0]]}]
        }]
      }]
    }]
   display_rate {numerator=60}
  }
 }
 Anim_Targeting {
  binding_count=1
  length_seconds=0.3
  object=&self:Anim_Targeting
  sequence {
   bindings [ReticleVisual {
     id=#4
     possessed_class=Image
     tracks [{
       type=MovieSceneWidgetMaterialTrack
       sections [{
         type=MovieSceneParameterSection
         channels [
          Targeting Circle Opacity {type=MovieSceneScriptingFloatChannel, keys [[0,0],[18,0.4]]}
          Triangle Size {type=MovieSceneScriptingFloatChannel, keys [[0,1],[18,2.5]]}
         ]
        }]
      }]
    }]
   display_rate {numerator=60}
  }
 }
 Anim_ChargeRotationUse {
  binding_count=1
  length_seconds=0.5
  object=&self:Anim_ChargeRotationUse
  sequence {bindings [StacksOverlay {
     id=#5
     possessed_class=Overlay
     tracks [{
       type=MovieScene2DTransformTrack
       sections [{
         type=MovieScene2DTransformSection
         end_frame=10
         start_frame=0
         channels [Angle {type=MovieSceneScriptingFloatChannel, keys [[0,0],[10,-24]]}]
        }]
      }]
    }]}
 }
]
graphs [
 EventGraph {nodes [
   n1 = UpdateAmmoCharges!() -> (OutputDelegate, then, NewLocalAmmo:int) {is=title K2Node_CustomEvent}
   > Set LocalAmmo(LocalAmmo:int=@n1.NewLocalAmmo) -> (Output_Get:int=0)
   > n2 = Branch(Condition="Equal (Integer)"(A=Get LocalAmmo() -> (LocalAmmo:int=0)))
   > n3 = Branch(Condition=Get DidUseAmmoBottom() -> (DidUseAmmoBottom:bool))
   OnUse(execute=@n3.else, self:WBP Reticle Hawaii Stack Object Reference=Get Stack03() -> (Stack03:WBP Reticle Hawaii Stack Object Reference))
   > Set DidUseAmmoBottom(DidUseAmmoBottom:bool=true) -> (Output_Get:bool)
   n4 = Branch(execute=@n2.else, Condition="Equal (Integer)"(A=Get LocalAmmo() -> (LocalAmmo=0), B=1))
   > n5 = Branch(Condition=Get DidUseAmmoMid() -> (DidUseAmmoMid:bool))
   OnUse(execute=@n5.else, self=Get Stack02() -> (Stack02:WBP Reticle Hawaii Stack Object Reference))
   > n6 = Set DidUseAmmoMid(DidUseAmmoMid:bool=true) -> (Output_Get:bool)
   Delay(execute=[@n6.then, @n8.then], Duration=0.6)
   > PlayAnimation(InAnimation=Get Anim_ChargeRotationUse() -> (Anim_ChargeRotationUse:Widget Animation Object Reference))
   Branch(execute=@n4.else, Condition="Equal (Integer)"(A=Get LocalAmmo() -> (LocalAmmo=0), B=2))
   > n7 = Branch(Condition=Get DidUseAmmoTop() -> (DidUseAmmoTop:bool))
   OnUse(execute=@n7.else, self=Get Stack01() -> (Stack01:WBP Reticle Hawaii Stack Object Reference))
   > n8 = Set DidUseAmmoTop(DidUseAmmoTop:bool=true) -> (Output_Get:bool)
   OnEquip() {is=title K2Node_CustomEvent}
   > ResetAmmoStates!(
    execute
    self
    NewAmmo:int=Get LocalBPAmmo!(self:B Weap Stamina Empty Object Reference=Get StaminaWeapon() -> (StaminaWeapon:B Weap Stamina Empty Object Reference)) -> (LocalBPAmmo:int=0)
   ) -> (then)
   n9 = ResetAmmoStates!() -> (OutputDelegate, then, NewAmmo:int) {is=title K2Node_CustomEvent}
   > OnInit(self:WBP Reticle Hawaii Stack Object Reference=[Get Stack01(), Get Stack02(), Get Stack03()])
   > Set DidUseAmmoTop(DidUseAmmoTop=false)
   > Set DidUseAmmoMid(DidUseAmmoMid=false)
   > Set DidUseAmmoBottom(DidUseAmmoBottom=false)
   > n10 = Branch(Condition="Equal (Integer)"(A=@n9.NewAmmo, B=1))
   > OnRecharge(self:WBP Reticle Hawaii Stack Object Reference=[Get Stack02(), Get Stack01()])
   > Set DidUseAmmoMid(DidUseAmmoMid=true)
   > Set DidUseAmmoTop(DidUseAmmoTop=true)
   n11 = Get StacksOverlay() -> (StacksOverlay:Overlay Object Reference)
   > SetRenderTransformAngle!(execute, self:Widget Object Reference=@n11, Angle:float=-24) -> (then)
   n12 = Branch(execute=@n10.else, Condition="Equal (Integer)"(A=@n9.NewAmmo, B=2))
   > OnRecharge(self=Get Stack01())
   > Set DidUseAmmoTop(DidUseAmmoTop=true)
   > SetRenderTransformAngle(self=@n11, Angle=0)
   SetRenderTransformAngle(execute=@n12.else, self=@n11, Angle=24)
   OnBeamFired() {is=title K2Node_CustomEvent}
   > Set ChargingUp(ChargingUp:bool=false) -> (Output_Get:bool)
   > SetVisibility(InVisibility=Collapsed)
   > StopAnimation(InAnimation=Get Anim_Charge() -> (Anim_Charge:Widget Animation Object Reference))
   AbilityFinished() {is=title K2Node_CustomEvent}
   > SetVisibility(InVisibility=HitTestInvisible)
   > n13 = GetDynamicMaterial(self=Get ReticleVisual() -> (ReticleVisual:Image Object Reference))
   > SetScalarParameterValue(self=@n13, ParameterName=Glyph Opacity, Value=SelectFloat(A=1, B=0, bPickA=Get IsGlyphShowing() -> (IsGlyphShowing:bool)))
   > SetScalarParameterValue(self=@n13, ParameterName=Global Opacity, Value=0.25)
   n14 = Get Anim_Charge()
   > PlayAnimation(InAnimation=@n14, StartAtTime=GetEndTime(self=@n14), PlayMode=Reverse)
   ChargeStarted() {is=title K2Node_CustomEvent}
   > UpdateAmmoCharges!(execute, self, NewLocalAmmo:int=Get LocalBPAmmo(self=Get StaminaWeapon()) -> (LocalBPAmmo=0)) -> (then) {is=title K2Node_CallFunction}
   > Set ChargingUp(ChargingUp=true)
   n15 = Get Anim_Charge()
   > PlayAnimation(InAnimation=Get Anim_Charge(), PlaybackSpeed=SafeDivide(A=GetEndTime(self=@n15), B=GetChargeDuration!() -> (ChargeDuration:double)))
   n16 = UpdateTargetInRange!() -> (OutputDelegate, then, bTargetInRange:bool=~) {is=title K2Node_CustomEvent}
   > Branch(Condition="Not Equal (Boolean)"(A=@n16.bTargetInRange, B=Get IsGlyphShowing()) -> (ReturnValue=~))
   > n17 = Branch(Condition=@n16.bTargetInRange)
   n18 = Get Anim_Targeting() -> (Anim_Targeting:Widget Animation Object Reference)
   > PlayAnimationForward(InAnimation=@n18)
   > Set IsGlyphShowing(IsGlyphShowing:bool=true) -> (Output_Get:bool)
   PlayAnimationReverse(execute=@n17.else, InAnimation=@n18)
   > Set IsGlyphShowing(IsGlyphShowing=false)
   n19 = HandleOnCooldown!() -> (OutputDelegate, then, bOnCooldown:bool=~) {is=title K2Node_CustomEvent}
   n20 = Get ChargingUp() -> (ChargingUp:bool)
   > n21 = Branch(Condition=AND Boolean(A=@n19.bOnCooldown, B=NOT Boolean(A=@n20)))
   n22 = GetDynamicMaterial(execute=@n21.else, self=Get ReticleVisual())
   > SetScalarParameterValue(self=@n22, ParameterName=Global Opacity, Value=1)
   > Branch(Condition=Get bWasOnCooldown() -> (bWasOnCooldown:bool))
   > PlayIntroState()
   > Set bWasOnCooldown(bWasOnCooldown:bool=false) -> (Output_Get:bool)
   Event Construct()
   > n23 = Sequence()
   > n24 = Cast To B_Weap_Stamina_Empty!(
    execute
    Object=GetEquippedWeapon!(self:Fort Weapon Reticle Extension Widget Base Object Reference) -> (ReturnValue:Fort Weapon Object Reference)
   ) -> (then, CastFailed, AsB Weap Stamina Empty:B Weap Stamina Empty Object Reference) {is=title K2Node_DynamicCast}
   > Set StaminaWeapon(StaminaWeapon:B Weap Stamina Empty Object Reference=@n24) -> (Output_Get:B Weap Stamina Empty Object Reference)
   > Bind Event to On Equipped(self:B Weap Stamina Empty Object Reference=@n24, Delegate=Create Event()) {is=title K2Node_AddDelegate}
   > Bind Event to Beam Fired Delegate(self:B Weap Stamina Empty Object Reference=@n24, Delegate=Create Event()) {is=title K2Node_AddDelegate}
   > Bind Event to Ability Finished(self:B Weap Stamina Empty Object Reference=@n24, Delegate=Create Event()) {is=title K2Node_AddDelegate}
   > Bind Event to Charge Started(self:B Weap Stamina Empty Object Reference=@n24, Delegate=Create Event()) {is=title K2Node_AddDelegate}
   > Delay(Duration=0.05)
   > n25 = Get!(execute:Exec, self:Actor Object Reference=@n24) -> (then:Exec, else:Exec, Instigator:Pawn Object Reference)
   n26 = GetComponentByClass!(
    self:Actor Object Reference=@n25
    ComponentClass:Actor Component Class Reference=/StaminaGameplay/Gameplay/Items/Hawaii/Stamina_Hawaii_Component.Stamina_Hawaii_Component_C
   ) -> (ReturnValue:Stamina Hawaii Component Object Reference)
   > Is Valid(InputObject=@n26)
   > Bind Event to On Has Target in Range(self:Stamina Hawaii Component Object Reference=@n26, Delegate=Create Event()) {is=title K2Node_AddDelegate}
   > Bind Event to On Cooldown State(self:Stamina Hawaii Component Object Reference=@n26, Delegate=Create Event()) {is=title K2Node_AddDelegate}
   PlayIntroState(execute=@n23.then_1)
   > OnEquip() {is=title K2Node_CallFunction}
  ]}
 GetChargeDuration {nodes [
   GetChargeDuration() {is=title K2Node_FunctionEntry}
   n1 = Get ChargeDurationData() -> (ChargeDurationData:Scalable Float Structure)
   > Return Node!(
    execute
    ChargeDuration:double=float + float(A=1.53, B=GetValueAtLevel!(Input:"Scalable Float Structure (by ref)"=@n1, Level:float=0) -> (ReturnValue:float))
   )
  ]}
 GetTargetSize {nodes [
   n1 = Inputs!() -> (execute, PhysicsObject:Actor Object Reference)
   > n2 = GetPhysicsObjectSimSize!(execute, PhysicsObject:Actor Object Reference=@n1) -> (then, ReturnValue:EFortPhysicsSimSize Enum=Small)
   n3 = Cast To FortPickup!(Object=@n1) -> (AsFort Pickup:Fort Pickup Object Reference, bSuccess:bool=~) {is=title K2Node_DynamicCast}
   n4 = Cast To FortProjectileBase!(Object=@n1) -> (AsFort Projectile Base:Fort Projectile Base Object Reference, bSuccess:bool=~) {is=title K2Node_DynamicCast}
   > Outputs!(
    then:Exec
    ReturnValue:EFortPhysicsSimSize Enum=Select(Option 0:EFortPhysicsSimSize Enum=@n2, Option 1:EFortPhysicsSimSize Enum=Small, Index=OR Boolean(A=@n3.bSuccess, B=@n4.bSuccess)) -> (ReturnValue:EFortPhysicsSimSize Enum)
   )
  ]}
 PlayIntroState {nodes [
   PlayIntroState() {is=title K2Node_FunctionEntry}
   > PlayAnimation(InAnimation=Get Anim_Intro() -> (Anim_Intro:Widget Animation Object Reference))
  ]}
 RE_StaminaHawaii_AutoGenFunc {nodes [RE_StaminaHawaii_AutoGenFunc() {is=title K2Node_FunctionEntry}]}
]
variables [
 bWasOnCooldown {type=boolean}
 ChargeDurationData {
  type=ScalableFloat
  default {
   registryType=None
   value=1
   curve {curveTable=/StaminaGameplay/DataTables/StaminaGameplayGameData, rowName=Default.Stamina.Hawaii.ChargeDuration}
  }
 }
 ChargingUp {type=boolean}
 DidUseAmmoBottom {type=boolean}
 DidUseAmmoMid {type=boolean}
 DidUseAmmoTop {type=boolean}
 IsGlyphShowing {type=boolean}
 LocalAmmo {type=integer}
 NewVar {type=ScalableFloat, default {registryType=None, value=0, curve {curveTable=null, rowName=None}}}
 StaminaWeapon {type=B_Weap_Stamina_Empty_C, default=null}
]
widgets {
 info {
  inheritedWidgetCount=1
  parentClass=/Script/FortniteUI.FortWeaponReticleExtensionWidgetBase
  rootWidgetClass=&Overlay
  widgetCount=12
 }
 widgets [
  Reticle {widgetClassPath=&Overlay, widget=Reticle}
  Overlay_0 {
   widgetClassPath=&Overlay
   parent=Reticle
   widget=Overlay_0
   slot {class=OverlaySlot, properties {horizontalAlignment=HAlign_Fill, verticalAlignment=VAlign_Fill}}
  }
  ReticleVisual {
   widgetClassPath=&Image
   bIsVariable=true
   parent=Overlay_0
   widget=ReticleVisual
   properties {brush {
     bIsDynamicallyLoaded=false
     drawAs=Image
     imageType=NoImage
     mirroring=NoMirror
     resourceName=None
     resourceObject=&/MI_UI_HawaiiReticle
     tiling=NoTile
     imageSize=<32 32>
     margin=<ltrb 0 0 0 0>
     outlineSettings {
      bUseBrushTransparency=false
      roundingType=HalfHeightRadius
      width=0
      color {colorUseRule=UseColor_Specified, specifiedColor=<rgba 0 0 0 0>}
      cornerRadii=<0 0 0 0>
     }
     tintColor {colorUseRule=UseColor_Specified, specifiedColor=<rgba 1 1 1 1>}
     uVRegion {bIsValid=false, max=<0 0>, min=<0 0>}
    }}
   slot {class=OverlaySlot, properties {horizontalAlignment=HAlign_Fill, verticalAlignment=VAlign_Fill}}
  }
  Spacer_205 {
   widgetClassPath=&Spacer
   parent=Overlay_0
   widget=Spacer_205
   properties {size=<96 96>}
   slot {class=OverlaySlot, properties {horizontalAlignment=HAlign_Fill, verticalAlignment=VAlign_Fill}}
  }
  StacksOverlay {
   widgetClassPath=&Overlay
   bIsVariable=true
   parent=Reticle
   widget=StacksOverlay
   properties {renderTransform {angle=24, scale=<1 1>, shear=<0 0>, translation=<0 0>}}
   slot {class=OverlaySlot, properties {horizontalAlignment=HAlign_Fill, verticalAlignment=VAlign_Fill}}
  }
  Stacks {
   widgetClassPath=&VerticalBox
   parent=StacksOverlay
   widget=Stacks
   properties {renderTransform {angle=0, scale=<1 1>, shear=<0 0>, translation=<76 0>}}
   slot {class=OverlaySlot, properties {horizontalAlignment=HAlign_Center, verticalAlignment=VAlign_Center}}
  }
  Stack01 {
   widgetClassPath=&/WBP_Reticle_Hawaii_Stack
   bIsVariable=true
   parent=Stacks
   widget=Stack01
   slot {class=VerticalBoxSlot}
  }
  Spacer_274 {widgetClassPath=&Spacer, parent=Stacks, widget=Spacer_274, properties {size=<1 2>}, slot {class=VerticalBoxSlot}}
  Stack02 {
   widgetClassPath=&WBP_Reticle_Hawaii_Stack
   bIsVariable=true
   parent=Stacks
   widget=Stack02
   properties {renderTransform {angle=0, scale=<1 1>, shear=<0 0>, translation=<8 0>}}
   slot {class=VerticalBoxSlot}
  }
  Spacer_356 {widgetClassPath=&Spacer, parent=Stacks, widget=Spacer_356, properties {size=<1 2>}, slot {class=VerticalBoxSlot}}
  Stack03 {
   widgetClassPath=&WBP_Reticle_Hawaii_Stack
   bIsVariable=true
   parent=Stacks
   widget=Stack03
   slot {class=VerticalBoxSlot}
  }
  TutorialHighlight {widgetClassPath=&TutorialHighlight}
 ]
}
